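; Quest gameplay ability that marks the owning pawn while it is at full health and shield, for a Battle Pass Season 16 elimination quest.
; On activation it binds a check to the pawn's health-changed event and runs it immediately.
; With both health and shield at 100 it applies a quest gameplay effect to the pawn.
; Otherwise it removes active effects granting the Athena.Quests.Elim.FullHealthShield tag.
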(event Ability|EventActivateAbility
  (|CallParentFunction)
  (Health|BindEventtoOnPawnHealthChanged (Variables|GAVars|Data|GetOwningPlayerPawn) (delegate Custom|Check%20Health))
  (CallFunction|CheckHealth))

(event Ability|EventFailedToActivatePassiveAbility
  (params (ActorInfo :type Gameplay%20Ability%20Actor%20Info%20Structure)))

(event Custom|Check%20Health
  (if (and (== (FortPawn|GetHealth (Variables|GAVars|Data|GetOwningPlayerPawn)) 1e+02) (== (FortPawn|GetShield (Variables|GAVars|Data|GetOwningPlayerPawn)) 1e+02))
    (GameplayEffects|ApplyGameplayEffectToSelf :self (Ability|GetAbilitySystemComponent (Variables|GAVars|Data|GetOwningPlayerPawn)) :GameplayEffectClass "/BattlepassS16/Items/QuestAbilities/Quest_Elim_FullHealthShield/GE_Quest_Elim_FullHealthShield.GE_Quest_Elim_FullHealthShield_C" :Level 0.0)
    (else
      (GameplayEffects|RemoveActiveEffectswithGrantedTags (Ability|GetAbilitySystemComponent (Variables|GAVars|Data|GetOwningPlayerPawn)) "(GameplayTags=((TagName=\"Athena.Quests.Elim.FullHealthShield\")))"))))
